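# Translucent, unlit, two-sided hologram material for Niagara mesh particles on the FriedNexus particle accelerator prop.
# Fresnel, vertex red and a sine-jittered scanner texture blend inner and outer colors, with a dynamic parameter picking friendly or enemy palettes.
# Depth-fade contact glow tints the color and is maxed into an opacity built from vertex and particle alpha, camera fades and an edge mask.
# A local sphere mask and panning world-space volume noise from a 2D slice atlas modulate that opacity.
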Material /Game/Effects/Fort_Effects/Effects/Environment/Hera/Props/FriedNexus_ParticleAccelerator_A/M_FriedNexus_Holo
graph {
 properties {
  bUsedWithNiagaraMeshParticles=true
  blendMode=BLEND_Translucent
  pixelDepthOffsetMode=PDOM_Legacy
  shadingModel=MSM_Unlit
  twoSided=true
  usedShadingModels=Unlit
 }
 nodes [
  n1 = ScalarParameter() {defaultValue=25, group=Contact Fade, parameterName=Contact Fade Distance, sortPriority=1}
  Add(A=@n1, B=Multiply(A=NamedRerouteUsage()) {constB=300})
  n2 = VectorParameter() {group=Color, parameterName=Color Inner Friendly, sortPriority=0, defaultValue=<rgba 1 0.5221 0.0734 1>}
  n3 = VectorParameter() {group=Color, parameterName=Color Outer Friendly, sortPriority=1, defaultValue=<rgba 10 6.8161 1.3167 1>}
  n4 = ScalarParameter() {defaultValue=1.3, group=Fresnel, parameterName=Fresnel Brightness, sortPriority=0}
  n5 = MaterialFunctionCall(Brightness=@n4, Exponent=ScalarParameter() {defaultValue=8, group=Fresnel, parameterName=Fresnel Exponent, sortPriority=1}) {materialFunction=&MF_Fresnel_FX}
  n6 = QualitySwitch(Default=@n5.Fresnel Inverted, Low=@n4, High=@n5.Fresnel Inverted, Medium=@n5.Fresnel Inverted)
  n7 = VertexColor()
  n8 = TextureCoordinate() {uTiling=8}
  n9 = DynamicParameter() {defaultValue=<rgba 0 0 0 0>, paramNames [ScannerRandomPosition, Speed Random, Param3, Team Color Picker]}
  n10 = AppendVector(A=@n9.ScannerRandomPosition, B=Constant())
  n11 = Multiply(A=Time(), B=ScalarParameter() {defaultValue=9, parameterName=Speed A})
  n12 = Sine(None=@n11) {period=1.3}
  n13 = Sine(None=@n11) {period=1.7}
  n14 = TextureSample!(UVs=Panner(Coordinate=Add(A=@n8, B=@n10), Time=Add(A=@n12, B=@n13)) {speedX=2.6}, Tex, Level, Apply View MipBias) {
   constMipValue=0
   mipValueMode=TMVM_MipLevel
   samplerType=SAMPLERTYPE_LinearColor
   texture=/Game/Effects/Fort_Effects/Textures/Tile/T_Raygun_BeamReveal_01
  }
  n15 = Clamp(None=Add(A=Add(A=@n6, B=@n7.R), B=@n14.R))
  n16 = VectorParameter() {group=Color, parameterName=Color Fade Friendly, sortPriority=2, defaultValue=<rgba 3 0.8897 0.3995 0>}
  n17 = VertexColor()
  n18 = VectorParameter() {group=Color, parameterName=Color Inner Enemy, sortPriority=3, defaultValue=<rgba 1 0.17 0.13 1>}
  n19 = VectorParameter() {group=Color, parameterName=Color Outer Enemy, sortPriority=4, defaultValue=<rgba 10 2.4663 1.3167 1>}
  n20 = VectorParameter() {group=Color, parameterName=Color Fade Friendly, sortPriority=2, defaultValue=<rgba 3 0.8897 0.3995 0>}
  n21 = DynamicParameter() {defaultValue=<rgba 0 0 0 0>, paramNames [ScannerRandomPosition, Speed Random, Param3, Team Color Picker]}
  n22 = LinearInterpolate(
   A=LinearInterpolate(A=LinearInterpolate(A=@n2.RGB, B=@n3.RGB, Alpha=@n15), B=@n16.RGB, Alpha=@n17.B)
   B=LinearInterpolate(A=LinearInterpolate(A=@n18.RGB, B=@n19.RGB, Alpha=@n15), B=@n20.RGB, Alpha=@n17.B)
   Alpha=@n21.Team Color Picker
  )
  n23 = VectorParameter() {group=Color, parameterName=Contact Fade Color, sortPriority=10, defaultValue=<rgba 0 1.7395 2 1>}
  n24 = NamedRerouteUsage()
  n25 = ScalarParameter() {defaultValue=10, group=Contact Fade, parameterName=Contact Noise Amplitude}
  n26 = DepthFade(FadeDistance=Add(A=@n1, B=Multiply(A=@n24, B=@n25) {constB=100}))
  n27 = ScalarParameter() {defaultValue=2, group=Contact Fade, parameterName=Contact Fade Exponent, sortPriority=2}
  n28 = Multiply(A=OneMinus(None=Power(Base=@n26, Exp=@n27)), B=ScalarParameter() {defaultValue=1, group=Contact Fade, parameterName=Contact Fade Opacity})
  n29 = ParticleColor()
  MaterialFunctionCall(Emissive=Multiply(A=LinearInterpolate(A=@n22, B=@n23.RGB, Alpha=@n28), B=@n29.RGB)) {materialFunction=&MF_FX_EmissiveAdjustments}
  n30 = LinearInterpolate(
   A=ScalarParameter() {defaultValue=0.35, group=Fresnel, parameterName=Fresnel Inner, sortPriority=2}
   B=ScalarParameter() {defaultValue=0.5, group=Fresnel, parameterName=Fresnel Outer, sortPriority=3}
   Alpha=@n15
  )
  n31 = Power(Base=VertexColor().A, Exp=ScalarParameter() {defaultValue=1, parameterName=GradientFadeExponent})
  n32 = Multiply(
   A=Multiply(A=@n29.A, B=Multiply(A=@n30, B=@n31))
   B=MaterialFunctionCall(Distance=ScalarParameter() {defaultValue=120, group=Distance Fade, parameterName=NearCameraFadeDistance, sortPriority=0}) {materialFunction=&Near_Camera_Fade}.Result
  )
  n33 = StaticSwitchParameter(
   True=MaterialFunctionCall(
    Opacity In=@n32
    Fade_Start_Distance=ScalarParameter() {defaultValue=1500, group=Distance Fade, parameterName=Distance Fade Start, sortPriority=1}
    Fade_End_Distance=ScalarParameter() {defaultValue=900, group=Distance Fade, parameterName=Distance Fade End, sortPriority=2}
   ) {materialFunction=&MF_Distance_SyncedFade_FX}.Result Inverted
   False=@n32
  ) {parameterName=UseDistanceSyncedFade?}
  n34 = TextureSampleParameter2D(UVs=TextureCoordinate()) {
   group=Edge Mask
   parameterName=MaskTexture
   samplerType=SAMPLERTYPE_LinearGrayscale
   sortPriority=0
   texture=/Game/Effects/Fort_Effects/Effects/Characters/Generic/ResIn/SquareGradient_Linear
  }
  n35 = ScalarParameter() {defaultValue=5, group=Edge Mask, parameterName=EdgeMaskScale, sliderMax=1, sortPriority=1}
  n36 = LinearInterpolate(
   A=@n33
   B=Multiply(A=@n33, B=Multiply(A=@n34.RGB, B=@n35) {constB=5})
   Alpha=ScalarParameter() {group=Edge Mask, parameterName=EdgeMaskWeight, sliderMax=1, sortPriority=2}
  )
  n37 = Multiply(
   A=LocalPosition().XYZ
   B=MaterialFunctionCall() {materialFunction=&ObjectScale}.Scale XYZ
  )
  n38 = Constant3Vector()
  n39 = ScalarParameter() {defaultValue=50, group=Sphere Mask, parameterName=Sphere Mask Radius, sortPriority=0}
  n40 = ScalarParameter() {group=Sphere Mask, parameterName=Sphere Mask Hardness, sortPriority=1}
  n41 = VectorParameter() {group=Volumetric Noise, parameterName=Noise Pan Speed, sortPriority=1, defaultValue=<rgba -0.25 -0.125 50 0>}
  n42 = Add(A=WorldPosition().XYZ, B=Multiply(A=Time(), B=@n41.RGB))
  n43 = ScalarParameter() {defaultValue=256, group=Volumetric Noise, parameterName=Noise Scale, sortPriority=0}
  n44 = ComponentMask(None=MaterialFunctionCall!(
    UVs=Divide(A=@n42, B=@n43) {constB=200}
    Slices Texture=TextureObject() {samplerType=SAMPLERTYPE_LinearColor, texture=/Game/Environments/Helios/TechArt/LowAltitudeClouds/T_LowAltitudeClouds_N}
    Atlas Size=Constant() {r=8}
    Atlas Padding=Constant() {r=0.9375}
   ) {materialFunction=/Game/Environments/Helios/TechArt/LowAltitudeClouds/MF_3DVolumeFrom2DSlices}.RGB) {r=true}
  Multiply(
   A=Multiply(A=Max(A=@n28, B=@n36), B=SphereMask(A=@n37, B=@n38, Radius=@n39, Hardness=@n40) {attenuationRadius=100, hardnessPercent=0})
   B=NamedRerouteDeclaration(None=Saturate(None=@n44)) {name=VolumeNoise, nodeColor=<rgba 0.1227 0.5676 1 1>}
  )
  NamedRerouteDeclaration(None=SphereMask(A=@n37, B=@n38) {attenuationRadius=200, hardnessPercent=0}) {name=InnerSphere, nodeColor=<rgba 0.1016 0 1 1>}
 ]
}
