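# Translucent, unlit, two-sided holographic wall material for the FriedNexus particle accelerator prop, usable on instanced meshes and Niagara mesh particles.
# World-space panning noise, sampled as a 3D volume from a 2D slice atlas, supplies animated variation.
# A noise-modulated depth fade highlights contact edges in color and opacity.
# Sphere masks on scale-corrected local position bound the opacity and blend inner and outer colors before emissive adjustment.
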Material /Game/Effects/Fort_Effects/Effects/Environment/Hera/Props/FriedNexus_ParticleAccelerator_A/M_FriedNexus_HoloWall
graph {
 properties {
  bUsedWithInstancedStaticMeshes=true
  bUsedWithNiagaraMeshParticles=true
  blendMode=BLEND_Translucent
  pixelDepthOffsetMode=PDOM_Legacy
  shadingModel=MSM_Unlit
  twoSided=true
  usedShadingModels=Unlit
 }
 nodes [
  NamedRerouteDeclaration(None=Multiply(
    A=LocalPosition().XYZ
    B=MaterialFunctionCall() {materialFunction=&ObjectScale}.Scale XYZ
   )) {name=SpherePosition, nodeColor=<rgba 0.8516 0 1 1>}
  NamedRerouteDeclaration(None=ScalarParameter() {defaultValue=50, group=Sphere Mask, parameterName=Sphere Mask Radius, sortPriority=0}) {name=SphereMaskRadius, nodeColor=<rgba 0.8516 0 1 1>}
  NamedRerouteDeclaration(None=Constant3Vector()) {name=SphereMaskCenter, nodeColor=<rgba 0.8516 0 1 1>}
  n1 = WorldPosition()
  n2 = VectorParameter() {group=Volumetric Noise, parameterName=Noise Pan Speed, sortPriority=1, defaultValue=<rgba -0.25 -0.125 50 0>}
  n3 = Multiply(A=Time(), B=@n2.RGB)
  n4 = MaterialFunctionCall!(
   UVs=Divide(A=Add(A=@n1.XYZ, B=@n3), B=ScalarParameter() {defaultValue=256, group=Volumetric Noise, parameterName=Noise Scale, sortPriority=0}) {constB=200}
   Slices Texture=TextureObject() {samplerType=SAMPLERTYPE_LinearColor, texture=/Game/Environments/Helios/TechArt/LowAltitudeClouds/T_LowAltitudeClouds_N}
   Atlas Size=Constant() {r=8}
   Atlas Padding=Constant() {r=0.9375}
  ) {materialFunction=/Game/Environments/Helios/TechArt/LowAltitudeClouds/MF_3DVolumeFrom2DSlices}
  NamedRerouteDeclaration(None=Saturate(None=ComponentMask(None=@n4.RGB) {r=true})) {name=VolumeNoise, nodeColor=<rgba 0.1227 0.5676 1 1>}
  n5 = NamedRerouteUsage()
  n6 = ScalarParameter() {defaultValue=10, group=Contact Fade, parameterName=Contact Noise Amplitude}
  n7 = DepthFade(FadeDistance=Add(
    A=ScalarParameter() {defaultValue=25, group=Contact Fade, parameterName=Contact Fade Distance, sortPriority=1}
    B=Multiply(A=@n5, B=@n6) {constB=100}
   ))
  n8 = ScalarParameter() {defaultValue=2, group=Contact Fade, parameterName=Contact Fade Exponent, sortPriority=2}
  n9 = Multiply(A=OneMinus(None=Power(Base=@n7, Exp=@n8)), B=ScalarParameter() {defaultValue=1, group=Contact Fade, parameterName=Contact Fade Opacity})
  n10 = SphereMask(A=NamedRerouteUsage(), B=NamedRerouteUsage(), Radius=NamedRerouteUsage(), Hardness=Constant() {r=1}) {attenuationRadius=100, hardnessPercent=0}
  n11 = Multiply(A=@n9, B=@n10)
  n12 = NamedRerouteUsage()
  n13 = Multiply(A=@n10, B=Multiply(A=OneMinus(None=@n12), B=ScalarParameter() {defaultValue=1, group=Opacity, parameterName=Opacity Scalar}))
  NamedRerouteDeclaration(None=Multiply(A=Max(A=@n11, B=@n13), B=NamedRerouteUsage())) {name=Opacity, nodeColor=<rgba 0 1 0.7656 1>}
  n14 = NamedRerouteUsage()
  n15 = NamedRerouteUsage()
  n16 = NamedRerouteUsage()
  n17 = ScalarParameter() {group=Color, parameterName=Color Grad Hardness, sortPriority=0}
  n18 = LinearInterpolate(
   A=VectorParameter() {group=Color, parameterName=Color Outer, sortPriority=1, defaultValue=<rgba 0.1227 0.5676 1 1>}.RGB
   B=VectorParameter() {group=Color, parameterName=Color Inner, sortPriority=2, defaultValue=<rgba 0.6278 0.0002 1 1>}.RGB
   Alpha=NamedRerouteDeclaration(None=SphereMask(A=@n14, B=@n15, Radius=@n16, Hardness=@n17)) {name=ColorGradMask, nodeColor=<rgba 0 1 0.3672 1>}
  )
  n19 = VectorParameter() {group=Color, parameterName=Contact Fade Color, sortPriority=10, defaultValue=<rgba 0 1.7395 2 1>}
  NamedRerouteDeclaration(None=MaterialFunctionCall(Emissive=LinearInterpolate(A=@n18, B=@n19.RGB, Alpha=@n9)) {materialFunction=&MF_FX_EmissiveAdjustments}.Result) {name=EmissiveColor, nodeColor=<rgba 0 1 0.4375 1>}
  NamedRerouteUsage()
  NamedRerouteUsage()
 ]
}
