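# Niagara effect for the FriedNexus particle accelerator prop's controller, adding a pulsing glow and radiation waves.
# Lighting bursts a persistent sprite with a waveform-pulsed light, picking green or red from a color array via the User.isInStorm bool.
# Radiation loops staggered bursts of randomly scaled meshes that grow and fade, with curve and random values driving dynamic material parameters.
# Only Lighting responds to the storm flag.
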NiagaraSystem /Game/Effects/Fort_Effects/Effects/Environment/Hera/Props/FriedNexus_ParticleAccelerator_A/NS_FriedNexus_Controller
dependencies {
 usedDataInterfaces [&NiagaraDataInterfaceCurve, &NiagaraDataInterfaceVectorCurve, &NiagaraDataInterfaceArrayColor]
 usedDynamicInputs []
 usedModules [
  &self:SystemSpawnScript.SystemScriptSource.SystemScriptGraph.NiagaraNodeAssignment_0.SetVariables_#1
  &SystemState
 ]
}
emitters {
 Lighting {
  emitterUpdateScript [
   &EmitterState()
   &SpawnBurst_Instantaneous(Use Spawn Probability=false, UseLoopCountLimit=true)
  ]
  particleSpawnScript [&InitializeParticle(
    Color=&SelectLinearColorFromArray(
     Array Sampling Mode=::NewEnumerator1
     Direct Array Index=&MakeCustomIntFromBool(Bool=$User.isInStorm, True INT=1, False INT=0)
     Color Selection Array=DataInterface() {propertyValues {ColorData [<rgba 0.1091 0.9 0.05 1>, <rgba 0.9 0.05 0.05 1>]}}
     Random Seed=?
    )
    Lifetime=5
    Mesh Scale Mode=::NewEnumerator3
    Sprite Size Mode=::NewEnumerator3
    Uniform Sprite Size=200
    UsePositionOffset=false
    Write Material Random=true
   )]
  particleUpdateScript [
   &ParticleState(Kill Particles When Lifetime Has Elapsed=false, Let Infinitely Lived Particles Die When Emitter Deactivates=true)
   &ScaleColor(
    Scale RGB=&VectorFromFloat(Value=&Waveform(Global Amplitude Scale=0.25, "Waveform [1]"=::NewEnumerator0, "[1] Frequency"=0.5, "[1] Amplitude Min/Max"=<0 1>, "[1] Bias"=0.7))
    ScaleA=true
    ScaleRGB=true
   )
   &Light_Attributes(
    Color=1
    Color Value To Scale:LinearColor=$Particles.Color
    Falloff Exponent=1
    Light Exponent=2
    Light Position Offset:Vector3f=<10 0 0>
    Light Radius=150
    Position=1
    Radius=1
    Scale RGB:Vector3f=&VectorFromFloat(Value=3)
    ScaleA=false
    ScaleRGB=true
    UsePositionOffset=true
   )
  ]
  rendererClasses [&NiagaraSpriteRendererProperties, &NiagaraLightRendererProperties]
  renderers [
   {rendererClass=&NiagaraSpriteRendererProperties}
   {rendererClass=&NiagaraLightRendererProperties, rendererIndex=1}
  ]
 }
 Radiation {
  emitterUpdateScript [
   &EmitterState(Life Cycle Mode=::NewEnumerator1, Loop Duration=1, UseLoopDelay=true)
   &SpawnBurst_Instantaneous(Spawn Time=0.5, Use Spawn Probability=false, UseLoopCountLimit=false) {module=SpawnBurst_Instantaneous001}
   &SpawnBurst_Instantaneous(Use Spawn Probability=false, UseLoopCountLimit=false)
  ]
  particleSpawnScript [&InitializeParticle(
    Color Mode=::NewEnumerator0
    Mesh Scale=/Niagara/DynamicInputs/CascadeConversion/SplitVector(
     Double Component Vector Channels:ECascadeNiagaraTwoVectorChannels=::NewEnumerator1
     Double Component:NiagaraFloat=&RandomRangeFloat(Minimum=7.5, Maximum=15)
     Single Component:NiagaraFloat=12.5
    )
    Mesh Scale Mode=::NewEnumerator4
    UsePositionOffset=false
    Write Material Random=true
   )]
  particleUpdateScript [
   &ParticleState()
   &ScaleMeshSize(Scale Factor=&VectorFromFloat(Value=&FloatFromCurve(
      FloatCurve=DataInterface() {propertyValues {Curve {defaultValue=1.3, keys [{tangentMode=RCTM_User, time=0, value=0.6}, {tangentMode=RCTM_User, time=1, value=1.3}]}}}
      Scale Curve=0.7
     )))
   &DynamicMaterialParameters(
    Index 0 Param 1=&FloatFromCurve(
     FloatCurve=DataInterface() {propertyValues {Curve {keys [
         {arriveTangent=0.0971, interpMode=RCIM_Cubic, time=0, value=0}
         {arriveTangent=1.4248, interpMode=RCIM_Cubic, leaveTangent=1.4248, time=0.2784, value=1}
         {arriveTangent=-1.3859, interpMode=RCIM_Cubic, leaveTangent=-1.3859, time=0.6737, value=0.9599}
         <at 1 0 RCIM_Cubic>
        ]}}}
     Scale Curve=0.3
    )
    Index 0 Param 2=&RandomRangeFloat(Minimum=-0.05, Maximum=-0.1)
    Index 0 Param 3=&RandomRangeFloat(Minimum=0.2, Maximum=1)
    Index 0 Param 4=&RandomRangeFloat(Minimum=0, Maximum=0.5)
    Module.Index0Float0Write=true
    Module.Index0Float1Write=true
    Module.Index0Float2Write=true
    Module.Index0Float3Write=true
   )
   &CameraOffset(Camera Offset Amount=25)
   &ScaleColor(
    Scale Alpha=&FloatFromCurve(FloatCurve=DataInterface() {propertyValues {Curve {keys [<at 0 0>,<at 0.3 1>,<at 1 0>]}}}, Scale Curve=0.4)
    Scale RGB=&VectorFromCurve(VectorCurve=DataInterface() {propertyValues {XCurve {keys [<at 0 1>,<at 1 1>]}, YCurve {keys [<at 0 1>,<at 1 1>]}, ZCurve {keys [<at 0 1>,<at 1 1>]}}})
    ScaleA=true
    ScaleRGB=false
   )
   &ScaleMeshSize(Scale Factor=&VectorFromFloat(Value=2.5)) {enabled=false, module=ScaleMeshSize001}
   &ScaleColor(Scale Alpha=2, ScaleA=true, ScaleRGB=true) {enabled=false, module=ScaleColor001}
  ]
  rendererClasses [&NiagaraMeshRendererProperties]
  renderers [{rendererClass=&NiagaraMeshRendererProperties, rendererIndex=0}]
 }
}
summary {
 systemName=NS_FriedNexus_Controller
 emitters [
  {emitterName=Radiation, rendererClasses [&NiagaraMeshRendererProperties]}
  {emitterName=Lighting, rendererClasses [&NiagaraSpriteRendererProperties, &NiagaraLightRendererProperties]}
 ]
 userVariables [User.isInStorm {type=NiagaraBool}]
}
