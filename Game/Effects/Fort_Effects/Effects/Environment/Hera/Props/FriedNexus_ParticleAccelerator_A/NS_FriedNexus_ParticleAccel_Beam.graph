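# Niagara system for the energy beam of the FriedNexus particle accelerator prop.
# Beam places a burst of looping-lifetime particles along a Hermite spline to System.EndPosition, drawn as sprites and ribbons with curve-shaped width.
# Inner and outer glow pairs mark both ends, the outer ones adding lights.
# Shimmers, tinted by User.InteriorColor and DestabilizeAmount, and curl-noise Sparkles surround the source, using camera-distance fading; SparklesEnd is disabled.
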NiagaraSystem /Game/Effects/Fort_Effects/Effects/Environment/Hera/Props/FriedNexus_ParticleAccelerator_A/NS_FriedNexus_ParticleAccel_Beam
dependencies {
 usedDataInterfaces [&NiagaraDataInterfaceCurve, &NiagaraDataInterfaceColorCurve, &NiagaraDataInterfaceStaticMesh]
 usedDynamicInputs []
 usedModules [
  &self:SystemSpawnScript.SystemScriptSource.SystemScriptGraph.NiagaraNodeAssignment_2.SetVariables_#1
  &SystemState
 ]
}
emitters {
 Beam {
  emitterUpdateScript [
   &EmitterState()
   &SpawnBurst_Instantaneous(Spawn Count=19, Use Spawn Probability=false, UseLoopCountLimit=true)
  ]
  particleSpawnScript [&InitializeParticle(
    Color=&LerpLinearColors(
     StartColor=<rgba 3.8703 1.0191 7 0.1>
     EndColor=<rgba 0 4.5537 5 0.5>
     LerpFactor=&FloatFromCurve(
      FloatCurve=DataInterface() {propertyValues {Curve {
         defaultValue=1
         keys [
          {interpMode=RCIM_Cubic, tangentMode=RCTM_User, time=0.24, value=0}
          {interpMode=RCIM_Cubic, tangentMode=RCTM_User, time=0.5, value=1}
          {tangentMode=RCTM_User, time=0.75, value=0}
         ]
        }}}
      CurveIndex=&ReturnNormalizedExecIndex()
     )
    )
    Lifetime=0.1
    Ribbon Width=80
    Ribbon Width Mode=::NewEnumerator1
    Sprite Size Mode=::NewEnumerator3
    Uniform Sprite Size=5
    UsePositionOffset=false
    Write Material Random=true
   )]
  particleUpdateScript [
   &ParticleState(
    Kill Particles When Lifetime Has Elapsed=false
    Let Infinitely Lived Particles Die When Emitter Deactivates=true
    Loop Particles Lifetime=true
   )
   /Game/Effects/Niagara/Modules/Curves/FN_HermiteSpline(
    Position Blend:NiagaraFloat=&ReturnNormalizedExecIndex()
    Position End:NiagaraPosition=$System.EndPosition
    Position Start:Vector3f=/Niagara/DynamicInputs/Helpers/SimulationPosition()
    Write Particles Position=true
   )
   &DynamicMaterialParameters(
    Index 0 Param 1=500
    Index 0 Param 2=&Multiply_Float(A=&OneMinusFloat(Float=$Emitter.LoopedAge), B=2)
    Module.Index0Float0Write=true
    Module.Index0Float1Write=true
    Module.Index0Float2Write=false
    Module.Index0Float3Write=true
   )
   &ScaleRibbonWidth(Ribbon Width Scale=&FloatFromCurve(
     FloatCurve=DataInterface() {propertyValues {Curve {
        defaultValue=1
        keys [
         {interpMode=RCIM_Cubic, tangentMode=RCTM_User, time=0, value=0.25}
         {interpMode=RCIM_Cubic, tangentMode=RCTM_User, time=0.5, value=1}
         {tangentMode=RCTM_User, time=1, value=0.25}
        ]
       }}}
     CurveIndex=&ReturnNormalizedExecIndex()
     Scale Curve=&Waveform(
      Global Input Position=&ReturnNormalizedExecIndex()
      "Waveform [1]"=::NewEnumerator0
      "[1] Phase Scale"=0.5
      "[1] Frequency"=2
      "[1] Amplitude Min/Max"=<0.8 1>
     )
    ))
  ]
  rendererClasses [&NiagaraSpriteRendererProperties, &NiagaraRibbonRendererProperties]
  renderers [
   {rendererClass=&NiagaraSpriteRendererProperties}
   {rendererClass=&NiagaraRibbonRendererProperties, rendererIndex=1}
   {rendererClass=&NiagaraRibbonRendererProperties, rendererIndex=2}
  ]
 }
 InnerGlow {
  emitterUpdateScript [
   &EmitterState()
   &SpawnRate(SpawnRate=10, Use Spawn Probability=true)
  ]
  particleSpawnScript [
   /SetVariables(Particles.EndColor:LinearColor=&ColorFromCurve(
     ColorCurve=DataInterface() {propertyValues {
       AlphaCurve {keys [<at 0 1 RCIM_Cubic>,<at 1 1 RCIM_Cubic>]}
       BlueCurve {keys [<at 0 1 RCIM_Cubic>,<at 1 1 RCIM_Cubic>]}
       GreenCurve {keys [<at 0 0.1029 RCIM_Cubic>,<at 1 0.0955 RCIM_Cubic>]}
       RedCurve {keys [<at 0 0.3609 RCIM_Cubic>,<at 1 0.1117 RCIM_Cubic>]}
      }}
     CurveIndex=&RandomRangeFloat(Minimum=0, Maximum=1)
    )) {isSetParameters=true, module=SetVariables_#2}
   /Niagara/Modules/Spawn/Initialization/InitializeParticle(
    Color=&ColorFromCurve(
     ColorCurve=DataInterface() {propertyValues {
       AlphaCurve {keys [<at 0 1 RCIM_Cubic>,<at 1 1 RCIM_Cubic>]}
       BlueCurve {keys [<at 0 1 RCIM_Cubic>,<at 1 1 RCIM_Cubic>]}
       GreenCurve {keys [<at 0 0.7152 RCIM_Cubic>,<at 1 0.2609 RCIM_Cubic>]}
       RedCurve {keys [<at 0 0.1856 RCIM_Cubic>,<at 1 0.0709 RCIM_Cubic>]}
      }}
     CurveIndex=&RandomRangeFloat(Minimum=0, Maximum=1)
    )
    Lifetime=&RandomRangeFloat(Minimum=0.3, Maximum=0.75)
    Module.Write Color=true
    Module.Write Lifetime=true
    Module.Write Mass=true
    Module.Write Material Random=true
    Module.Write Position=false
    Module.Write Scale=false
    Module.Write SpriteRotation=false
    Module.Write SpriteSize=true
    Position=/Niagara/Functions/Localspace/SimulationPosition()
    Sprite Size=&Vector2DFromFloat(Value=15)
   ) {enabled=false}
   /Niagara/Modules/Spawn/Initialization/V2/InitializeParticle(
    Color=&ColorFromCurve(
     ColorCurve=DataInterface() {propertyValues {
       AlphaCurve {keys [<at 0 1 RCIM_Cubic>,<at 1 1 RCIM_Cubic>]}
       BlueCurve {keys [<at 0 1 RCIM_Cubic>,<at 1 1 RCIM_Cubic>]}
       GreenCurve {keys [<at 0 0.7152 RCIM_Cubic>,<at 1 0.2609 RCIM_Cubic>]}
       RedCurve {keys [<at 0 0.1856 RCIM_Cubic>,<at 1 0.0709 RCIM_Cubic>]}
      }}
     CurveIndex=&RandomRangeFloat(Minimum=0, Maximum=1)
    )
    Lifetime=&RandomRangeFloat(Minimum=0.3, Maximum=0.75)
    Sprite Size Mode=::NewEnumerator3
    Uniform Sprite Size=70
    UsePositionOffset=false
    Write Material Random=true
   ) {module=InitializeParticle001}
   &SphereLocation(
    MaskSpawnGroup=false
    Offset=$System.DTSphere_Offset
    Override Seed=false
    Sphere Radius=0
    Surface Only=false
    UseNonUniformScale=false
    UseOffset=true
    UseSphereOrientation=false
   )
   &DynamicMaterialParameters(
    Index 0 Param 1=100
    Module.Param0WriteEnabled=true
    Module.Param1WriteEnabled=false
    Module.Param2WriteEnabled=false
    Module.Param3WriteEnabled=false
   )
   &CameraOffset(Camera Offset Amount=25)
  ]
  particleUpdateScript [
   &ParticleState()
   &ScaleColor(
    Scale Alpha=&FloatFromCurve(
     FloatCurve=DataInterface() {propertyValues {Curve {keys [{interpMode=RCIM_Cubic, tangentMode=RCTM_User, time=0, value=1}, <at 1 0>]}}}
     Scale Curve=100
    )
    Scale RGB=&Multiply_VectorByFloat(
     Vector=<1 1 1>
     Float=&FloatFromCurve(
      FloatCurve=DataInterface() {propertyValues {Curve {keys [{interpMode=RCIM_Cubic, tangentMode=RCTM_User, time=0, value=0.2}, <at 1 1>]}}}
      Scale Curve=&RandomRangeFloat(Minimum=5, Maximum=200)
     )
    )
    ScaleA=true
    ScaleRGB=false
   )
   &ScaleSpriteSize(Scale Factor=&Vector2DFromFloat(Value=&FloatFromCurve(FloatCurve=DataInterface() {propertyValues {Curve {keys [{arriveTangent=3.2933, interpMode=RCIM_Cubic, leaveTangent=3.2933, tangentMode=RCTM_User, time=0, value=0.5}, <at 1 1>]}}})))
   /SetVariables(Particles.Initial.Color:LinearColor=&LerpLinearColors(
     StartColor=$Particles.Initial.Color
     EndColor=$Particles.EndColor
     LerpFactor=&FloatFromCurve(FloatCurve=DataInterface() {propertyValues {Curve {keys [
          {interpMode=RCIM_Cubic, tangentMode=RCTM_User, time=0, value=0}
          {arriveTangent=0.955, interpMode=RCIM_Cubic, leaveTangent=0.955, tangentMode=RCTM_User, time=1, value=1}
         ]}}})
    )) {isSetParameters=true, module=SetVariables_#3}
  ]
  rendererClasses [&NiagaraSpriteRendererProperties]
  renderers [
   {rendererClass=&NiagaraSpriteRendererProperties}
   {rendererClass=&NiagaraSpriteRendererProperties, rendererIndex=1}
  ]
 }
 InnerGlow001 {
  emitterUpdateScript [
   &EmitterState()
   &SpawnRate(SpawnRate=10, Use Spawn Probability=true)
  ]
  particleSpawnScript [
   /SetVariables(Particles.EndColor:LinearColor=&ColorFromCurve(
     ColorCurve=DataInterface() {propertyValues {
       AlphaCurve {keys [<at 0 1 RCIM_Cubic>,<at 1 1 RCIM_Cubic>]}
       BlueCurve {keys [<at 0 1 RCIM_Cubic>,<at 1 1 RCIM_Cubic>]}
       GreenCurve {keys [<at 0 0.1029 RCIM_Cubic>,<at 1 0.0955 RCIM_Cubic>]}
       RedCurve {keys [<at 0 0.3609 RCIM_Cubic>,<at 1 0.1117 RCIM_Cubic>]}
      }}
     CurveIndex=&RandomRangeFloat(Minimum=0, Maximum=1)
    )) {isSetParameters=true, module=SetVariables_#2}
   /Niagara/Modules/Spawn/Initialization/InitializeParticle(
    Color=&ColorFromCurve(
     ColorCurve=DataInterface() {propertyValues {
       AlphaCurve {keys [<at 0 1 RCIM_Cubic>,<at 1 1 RCIM_Cubic>]}
       BlueCurve {keys [<at 0 1 RCIM_Cubic>,<at 1 1 RCIM_Cubic>]}
       GreenCurve {keys [<at 0 0.7152 RCIM_Cubic>,<at 1 0.2609 RCIM_Cubic>]}
       RedCurve {keys [<at 0 0.1856 RCIM_Cubic>,<at 1 0.0709 RCIM_Cubic>]}
      }}
     CurveIndex=&RandomRangeFloat(Minimum=0, Maximum=1)
    )
    Lifetime=&RandomRangeFloat(Minimum=0.3, Maximum=0.75)
    Module.Write Color=true
    Module.Write Lifetime=true
    Module.Write Mass=true
    Module.Write Material Random=true
    Module.Write Position=false
    Module.Write Scale=false
    Module.Write SpriteRotation=false
    Module.Write SpriteSize=true
    Position=&SimulationPosition()
    Sprite Size=&Vector2DFromFloat(Value=15)
   ) {enabled=false}
   /Niagara/Modules/Spawn/Initialization/V2/InitializeParticle(
    Color=&ColorFromCurve(
     ColorCurve=DataInterface() {propertyValues {
       AlphaCurve {keys [<at 0 1 RCIM_Cubic>,<at 1 1 RCIM_Cubic>]}
       BlueCurve {keys [<at 0 1 RCIM_Cubic>,<at 1 1 RCIM_Cubic>]}
       GreenCurve {keys [<at 0 0.7152 RCIM_Cubic>,<at 1 0.2609 RCIM_Cubic>]}
       RedCurve {keys [<at 0 0.1856 RCIM_Cubic>,<at 1 0.0709 RCIM_Cubic>]}
      }}
     CurveIndex=&RandomRangeFloat(Minimum=0, Maximum=1)
    )
    Lifetime=&RandomRangeFloat(Minimum=0.3, Maximum=0.75)
    Position Offset:NiagaraPosition=$System.EndPosition
    Sprite Size Mode=::NewEnumerator3
    Uniform Sprite Size=100
    UsePositionOffset=true
    Write Material Random=true
   ) {module=InitializeParticle001}
   &SphereLocation(
    MaskSpawnGroup=false
    Override Seed=false
    Sphere Radius=0
    Surface Only=false
    UseNonUniformScale=false
    UseOffset=false
    UseSphereOrientation=false
   )
   &DynamicMaterialParameters(
    Index 0 Param 1=100
    Module.Param0WriteEnabled=true
    Module.Param1WriteEnabled=false
    Module.Param2WriteEnabled=false
    Module.Param3WriteEnabled=false
   )
   &CameraOffset(Camera Offset Amount=25)
  ]
  particleUpdateScript [
   &ParticleState()
   &ScaleColor(
    Scale Alpha=&FloatFromCurve(
     FloatCurve=DataInterface() {propertyValues {Curve {keys [{interpMode=RCIM_Cubic, tangentMode=RCTM_User, time=0, value=1}, <at 1 0>]}}}
     Scale Curve=100
    )
    Scale RGB=&Multiply_VectorByFloat(
     Vector=<1 1 1>
     Float=&FloatFromCurve(
      FloatCurve=DataInterface() {propertyValues {Curve {keys [{interpMode=RCIM_Cubic, tangentMode=RCTM_User, time=0, value=0.2}, <at 1 1>]}}}
      Scale Curve=&RandomRangeFloat(Minimum=5, Maximum=200)
     )
    )
    ScaleA=true
    ScaleRGB=false
   )
   &ScaleSpriteSize(Scale Factor=&Vector2DFromFloat(Value=&FloatFromCurve(FloatCurve=DataInterface() {propertyValues {Curve {keys [{arriveTangent=3.2933, interpMode=RCIM_Cubic, leaveTangent=3.2933, tangentMode=RCTM_User, time=0, value=0.5}, <at 1 1>]}}})))
   /SetVariables(Particles.Initial.Color:LinearColor=&LerpLinearColors(
     StartColor=$Particles.Initial.Color
     EndColor=$Particles.EndColor
     LerpFactor=&FloatFromCurve(FloatCurve=DataInterface() {propertyValues {Curve {keys [
          {interpMode=RCIM_Cubic, tangentMode=RCTM_User, time=0, value=0}
          {arriveTangent=0.955, interpMode=RCIM_Cubic, leaveTangent=0.955, tangentMode=RCTM_User, time=1, value=1}
         ]}}})
    )) {isSetParameters=true, module=SetVariables_#3}
  ]
  rendererClasses [&NiagaraSpriteRendererProperties]
  renderers [{rendererClass=&NiagaraSpriteRendererProperties}]
 }
 OuterGlow {
  emitterUpdateScript [
   &EmitterState(Enable Distance Culling=true, MaxDistance=6000, Module.bUseMaxDistance=true, Module.bUseMinDistance=false, Scalability Mode=::NewEnumerator1)
   &SpawnRate(SpawnRate=1.5, Use Spawn Probability=true)
  ]
  particleSpawnScript [
   /SetVariables(Particles.EndColor:LinearColor=&ColorFromCurve(
     ColorCurve=DataInterface() {propertyValues {
       AlphaCurve {keys [<at 0 1 RCIM_Cubic>,<at 1 1 RCIM_Cubic>]}
       BlueCurve {keys [<at 0 1 RCIM_Cubic>,<at 1 1 RCIM_Cubic>]}
       GreenCurve {keys [<at 0 0.1029 RCIM_Cubic>,<at 1 0.0955 RCIM_Cubic>]}
       RedCurve {keys [<at 0 0.3609 RCIM_Cubic>,<at 1 0.1117 RCIM_Cubic>]}
      }}
     CurveIndex=&RandomRangeFloat(Minimum=0, Maximum=1)
    )) {isSetParameters=true, module=SetVariables_#2}
   /Niagara/Modules/Spawn/Initialization/InitializeParticle(
    Color=&ColorFromCurve(
     ColorCurve=DataInterface() {propertyValues {
       AlphaCurve {keys [<at 0 1 RCIM_Cubic>,<at 1 1 RCIM_Cubic>]}
       BlueCurve {keys [<at 0 1 RCIM_Cubic>,<at 1 1 RCIM_Cubic>]}
       GreenCurve {keys [<at 0 0.7152 RCIM_Cubic>,<at 1 0.2609 RCIM_Cubic>]}
       RedCurve {keys [<at 0 0.1856 RCIM_Cubic>,<at 1 0.0709 RCIM_Cubic>]}
      }}
     CurveIndex=&RandomRangeFloat(Minimum=0, Maximum=1)
    )
    Lifetime=&RandomRangeFloat(Minimum=1, Maximum=1.5)
    Module.Write Color=true
    Module.Write Lifetime=true
    Module.Write Mass=true
    Module.Write Material Random=true
    Module.Write Position=false
    Module.Write Scale=false
    Module.Write SpriteRotation=true
    Module.Write SpriteSize=true
    Position=&SimulationPosition()
    Sprite Rotation=&RandomRangeFloat(Minimum=0, Maximum=1)
    Sprite Size=&Multiply_Vector2(
     A=&RandomRangeVector2D(Minimum=<0.75 0.75>, Maximum=<1 1>)
     B=&Vector2DFromFloat(Value=50)
    )
   ) {enabled=false}
   /Niagara/Modules/Spawn/Initialization/V2/InitializeParticle(
    Color=&ColorFromCurve(
     ColorCurve=DataInterface() {propertyValues {
       AlphaCurve {keys [<at 0 1 RCIM_Cubic>,<at 1 1 RCIM_Cubic>]}
       BlueCurve {keys [<at 0 1 RCIM_Cubic>,<at 1 1 RCIM_Cubic>]}
       GreenCurve {keys [<at 0 0.7152 RCIM_Cubic>,<at 1 0.2609 RCIM_Cubic>]}
       RedCurve {keys [<at 0 0.1856 RCIM_Cubic>,<at 1 0.0709 RCIM_Cubic>]}
      }}
     CurveIndex=&RandomRangeFloat(Minimum=0, Maximum=1)
    )
    Lifetime=&RandomRangeFloat(Minimum=1, Maximum=1.5)
    Sprite Size Mode=::NewEnumerator3
    Uniform Sprite Size=100
    UsePositionOffset=false
    Write Material Random=false
   ) {module=InitializeParticle001}
   &SphereLocation(
    MaskSpawnGroup=false
    Offset=$System.DTSphere_Offset
    Override Seed=false
    Sphere Radius=0
    Surface Only=false
    UseNonUniformScale=false
    UseOffset=true
    UseSphereOrientation=false
   )
   &DynamicMaterialParameters(
    Index 0 Param 1=100
    Module.Param0WriteEnabled=true
    Module.Param1WriteEnabled=false
    Module.Param2WriteEnabled=false
    Module.Param3WriteEnabled=false
   )
   /SetVariables(Particles.SpriteSizeDecal:Vector2f=&Vector2DFromFloat(Value=200)) {isSetParameters=true}
  ]
  particleUpdateScript [
   &ParticleState()
   &ScaleColor(
    Scale Alpha=&FloatFromCurve(
     FloatCurve=DataInterface() {propertyValues {Curve {keys [{interpMode=RCIM_Cubic, tangentMode=RCTM_User, time=0, value=1}, <at 1 0>]}}}
     Scale Curve=&FloatFromCurve(
      FloatCurve=DataInterface() {propertyValues {Curve {keys [
          <at 0 0.03>
          {arriveTangent=0.0105, interpMode=RCIM_Cubic, leaveTangent=0.0105, tangentMode=RCTM_User, time=0.1, value=0.03}
          {arriveTangent=0.4347, interpMode=RCIM_Cubic, leaveTangent=-0.0047, tangentMode=RCTM_Break, time=0.5, value=0.1}
          <at 1 0.1>
         ]}}}
      CurveIndex=$Emitter.Scalability.DistanceFraction
     )
    )
    Scale RGB=&Multiply_VectorByFloat(
     Vector=<1 1 1>
     Float=&FloatFromCurve(
      FloatCurve=DataInterface() {propertyValues {Curve {keys [{interpMode=RCIM_Cubic, tangentMode=RCTM_User, time=0, value=0.2}, <at 1 1>]}}}
      Scale Curve=50
     )
    )
    ScaleA=true
    ScaleRGB=true
   )
   &ScaleSpriteSize(Scale Factor=&Vector2DFromFloat(Value=&FloatFromCurve(FloatCurve=DataInterface() {propertyValues {Curve {keys [{arriveTangent=1.4416, interpMode=RCIM_Cubic, leaveTangent=1.4416, tangentMode=RCTM_User, time=0, value=0.9}, <at 1 1>]}}})))
   /SetVariables(Particles.Initial.Color:LinearColor=&LerpLinearColors(
     StartColor=$Particles.Initial.Color
     EndColor=$Particles.EndColor
     LerpFactor=&FloatFromCurve(FloatCurve=DataInterface() {propertyValues {Curve {keys [
          {interpMode=RCIM_Cubic, tangentMode=RCTM_User, time=0, value=0}
          {arriveTangent=0.955, interpMode=RCIM_Cubic, leaveTangent=0.955, tangentMode=RCTM_User, time=1, value=1}
         ]}}})
    )) {isSetParameters=true, module=SetVariables_#3}
   &Light_Attributes(Color=1, Scale RGB:Vector3f=&VectorFromFloat(Value=100), ScaleA=true, ScaleRGB=true)
  ]
  rendererClasses [&NiagaraSpriteRendererProperties, &NiagaraLightRendererProperties]
  renderers [
   {rendererClass=&NiagaraSpriteRendererProperties}
   {rendererClass=&NiagaraLightRendererProperties, rendererIndex=1}
  ]
 }
 OuterGlow001 {
  emitterUpdateScript [
   &EmitterState(Enable Distance Culling=true, MaxDistance=6000, Module.bUseMaxDistance=true, Module.bUseMinDistance=false, Scalability Mode=::NewEnumerator1)
   &SpawnRate(SpawnRate=1.5, Use Spawn Probability=true)
  ]
  particleSpawnScript [
   /SetVariables(Particles.EndColor:LinearColor=&ColorFromCurve(
     ColorCurve=DataInterface() {propertyValues {
       AlphaCurve {keys [<at 0 1 RCIM_Cubic>,<at 1 1 RCIM_Cubic>]}
       BlueCurve {keys [<at 0 1 RCIM_Cubic>,<at 1 1 RCIM_Cubic>]}
       GreenCurve {keys [<at 0 0.1029 RCIM_Cubic>,<at 1 0.0955 RCIM_Cubic>]}
       RedCurve {keys [<at 0 0.3609 RCIM_Cubic>,<at 1 0.1117 RCIM_Cubic>]}
      }}
     CurveIndex=&RandomRangeFloat(Minimum=0, Maximum=1)
    )) {isSetParameters=true, module=SetVariables_#2}
   /Niagara/Modules/Spawn/Initialization/InitializeParticle(
    Color=&ColorFromCurve(
     ColorCurve=DataInterface() {propertyValues {
       AlphaCurve {keys [<at 0 1 RCIM_Cubic>,<at 1 1 RCIM_Cubic>]}
       BlueCurve {keys [<at 0 1 RCIM_Cubic>,<at 1 1 RCIM_Cubic>]}
       GreenCurve {keys [<at 0 0.7152 RCIM_Cubic>,<at 1 0.2609 RCIM_Cubic>]}
       RedCurve {keys [<at 0 0.1856 RCIM_Cubic>,<at 1 0.0709 RCIM_Cubic>]}
      }}
     CurveIndex=&RandomRangeFloat(Minimum=0, Maximum=1)
    )
    Lifetime=&RandomRangeFloat(Minimum=1, Maximum=1.5)
    Module.Write Color=true
    Module.Write Lifetime=true
    Module.Write Mass=true
    Module.Write Material Random=true
    Module.Write Position=false
    Module.Write Scale=false
    Module.Write SpriteRotation=true
    Module.Write SpriteSize=true
    Position=&SimulationPosition()
    Sprite Rotation=&RandomRangeFloat(Minimum=0, Maximum=1)
    Sprite Size=&Multiply_Vector2(
     A=&RandomRangeVector2D(Minimum=<0.75 0.75>, Maximum=<1 1>)
     B=&Vector2DFromFloat(Value=50)
    )
   ) {enabled=false}
   /Niagara/Modules/Spawn/Initialization/V2/InitializeParticle(
    Color=&ColorFromCurve(
     ColorCurve=DataInterface() {propertyValues {
       AlphaCurve {keys [<at 0 1 RCIM_Cubic>,<at 1 1 RCIM_Cubic>]}
       BlueCurve {keys [<at 0 1 RCIM_Cubic>,<at 1 1 RCIM_Cubic>]}
       GreenCurve {keys [<at 0 0.7152 RCIM_Cubic>,<at 1 0.2609 RCIM_Cubic>]}
       RedCurve {keys [<at 0 0.1856 RCIM_Cubic>,<at 1 0.0709 RCIM_Cubic>]}
      }}
     CurveIndex=&RandomRangeFloat(Minimum=0, Maximum=1)
    )
    Lifetime=&RandomRangeFloat(Minimum=1, Maximum=1.5)
    Position Offset:NiagaraPosition=$System.EndPosition
    Sprite Size Mode=::NewEnumerator3
    Uniform Sprite Size=100
    UsePositionOffset=true
    Write Material Random=true
   ) {module=InitializeParticle001}
   &SphereLocation(
    MaskSpawnGroup=false
    Override Seed=false
    Sphere Radius=0
    Surface Only=false
    UseNonUniformScale=false
    UseOffset=true
    UseSphereOrientation=false
   )
   &DynamicMaterialParameters(
    Index 0 Param 1=100
    Module.Param0WriteEnabled=true
    Module.Param1WriteEnabled=false
    Module.Param2WriteEnabled=false
    Module.Param3WriteEnabled=false
   )
   &CameraOffset(Camera Offset Amount=25)
  ]
  particleUpdateScript [
   &ParticleState()
   &ScaleColor(
    Scale Alpha=&FloatFromCurve(
     FloatCurve=DataInterface() {propertyValues {Curve {keys [{interpMode=RCIM_Cubic, tangentMode=RCTM_User, time=0, value=1}, <at 1 0>]}}}
     Scale Curve=&FloatFromCurve(
      FloatCurve=DataInterface() {propertyValues {Curve {keys [
          <at 0 0.03>
          {arriveTangent=0.0105, interpMode=RCIM_Cubic, leaveTangent=0.0105, tangentMode=RCTM_User, time=0.1, value=0.03}
          {arriveTangent=0.4347, interpMode=RCIM_Cubic, leaveTangent=-0.0047, tangentMode=RCTM_Break, time=0.5, value=0.1}
          <at 1 0.1>
         ]}}}
      CurveIndex=$Emitter.Scalability.DistanceFraction
     )
    )
    Scale RGB=&Multiply_VectorByFloat(
     Vector=<1 1 1>
     Float=&FloatFromCurve(
      FloatCurve=DataInterface() {propertyValues {Curve {keys [{interpMode=RCIM_Cubic, tangentMode=RCTM_User, time=0, value=0.2}, <at 1 1>]}}}
      Scale Curve=50
     )
    )
    ScaleA=true
    ScaleRGB=true
   )
   &ScaleSpriteSize(Scale Factor=&Vector2DFromFloat(Value=&FloatFromCurve(FloatCurve=DataInterface() {propertyValues {Curve {keys [{arriveTangent=1.4416, interpMode=RCIM_Cubic, leaveTangent=1.4416, tangentMode=RCTM_User, time=0, value=0.9}, <at 1 1>]}}})))
   /SetVariables(Particles.Initial.Color:LinearColor=&LerpLinearColors(
     StartColor=$Particles.Initial.Color
     EndColor=$Particles.EndColor
     LerpFactor=&FloatFromCurve(FloatCurve=DataInterface() {propertyValues {Curve {keys [
          {interpMode=RCIM_Cubic, tangentMode=RCTM_User, time=0, value=0}
          {arriveTangent=0.955, interpMode=RCIM_Cubic, leaveTangent=0.955, tangentMode=RCTM_User, time=1, value=1}
         ]}}})
    )) {isSetParameters=true, module=SetVariables_#3}
   &Light_Attributes(Color=1, Scale RGB:Vector3f=&VectorFromFloat(Value=100), ScaleA=true, ScaleRGB=true)
  ]
  rendererClasses [&NiagaraSpriteRendererProperties, &NiagaraLightRendererProperties]
  renderers [
   {rendererClass=&NiagaraSpriteRendererProperties}
   {rendererClass=&NiagaraLightRendererProperties, rendererIndex=1}
  ]
 }
 Shimmers {
  emitterUpdateScript [
   &EmitterState(Enable Distance Culling=true, MaxDistance=7000, Module.bUseMaxDistance=true, Module.bUseMinDistance=false, Scalability Mode=::NewEnumerator1)
   &SpawnRate(SpawnRate=5, Use Spawn Probability=false)
  ]
  particleSpawnScript [
   /SetVariables(Particles.ParticleVelocity=5000) {isSetParameters=true}
   /SetVariables(Particles.VelFromPointRand:NiagaraFloat=&FloatFromCurve(
     FloatCurve=DataInterface() {propertyValues {Curve {keys [
         {interpMode=RCIM_Cubic, tangentMode=RCTM_User, time=0, value=0.1}
         {arriveTangent=2.7242, interpMode=RCIM_Cubic, leaveTangent=2.7242, tangentMode=RCTM_User, time=1, value=1}
        ]}}}
     CurveIndex=&RandomRangeFloat(Minimum=0, Maximum=1)
     Scale Curve=$Particles.ParticleVelocity
    )) {isSetParameters=true}
   &InitializeParticle(
    Color=&LerpLinearColors(
     StartColor=$User.InteriorColor
     EndColor=<rgba 0.9714 0.5 1 0.25>
     LerpFactor=&Lerp_Float(A=0.7, B=1, Alpha=$System.DestabilizeAmount)
    )
    Lifetime Max=5
    Lifetime Min=4
    Lifetime Mode=::NewEnumerator1
    Sprite Rotation Mode=::NewEnumerator2
    Sprite Size Mode=::NewEnumerator1
    Uniform Sprite Size Max=110
    Uniform Sprite Size Min=90
    UsePositionOffset=true
    Write Material Random=true
   )
   /Niagara/Modules/Spawn/MeshInterface/SampleStaticMesh(
    OverrideRandomSeed=false
    StaticMesh=DataInterface() {propertyValues {
      DefaultMesh=/Game/Effects/Fort_Effects/Effects/Environment/DarkTower/SM_DarkTower_ZPContainment_Cracks
      InstanceIndex=-1
      LODIndex=0
      PreviewMesh=null
      SoftSourceActor=null
      SourceMode=DefaultMeshOnly
      bAllowSamplingFromStreamingLODs=false
      bCaptureTransformsPerFrame=true
      bUsePhysicsBodyVelocity=false
      FilteredSockets []
      LODIndexUserParameter {parameter {name=None, typeDefHandle {registeredTypeIndex=108}, varData []}}
      MeshParameterBinding {parameter {name=None, typeDefHandle {registeredTypeIndex=1}, varData []}}
      SectionFilter {allowedMaterialSlots []}
     }}
   ) {enabled=false}
   /Niagara/Modules/Spawn/Location/StaticMeshLocation(
    Module.Enable Axis Constraint=false
    Sampled Mesh Position:NiagaraPosition=&Add_Vector(A=Linked() {linkedVariable {name=Particles.SampleStaticMesh.MeshPosition, type=NiagaraPosition}}, B=$System.DTSphere_Offset)
   ) {enabled=false}
   &AddVelocityFromPoint(Module.Use Falloff Distance=false, Origin Offset=<0 -10 0>, Velocity Strength=200)
   &CameraOffset(Camera Offset Amount=80)
  ]
  particleUpdateScript [
   &ParticleState()
   &DynamicMaterialParameters(
    Index 0 Param 1=$Particles.MaterialRandom
    Index 0 Param 2=&RandomRangeFloat(Minimum=0, Maximum=1)
    Index 0 Param 3=&Lerp_Float(A=0, B=2, Alpha=$System.DestabilizeAmount)
    Module.Param0WriteEnabled=true
    Module.Param1WriteEnabled=true
    Module.Param2WriteEnabled=true
    Module.Param3WriteEnabled=false
   )
   &Drag(Drag=0.4, Use Linear Drag=true, Use Rotational Drag=false)
   &GravityForce(Gravity=&RandomRangeVector(Minimum=<0 0 -4>, Maximum=<1 1 -20>)) {enabled=false}
   &ScaleColor(
    Scale Alpha=&FloatFromCurve(
     FloatCurve=DataInterface() {propertyValues {Curve {keys [<at 0 0 RCIM_Cubic>,<at 0.35 1 RCIM_Cubic>,<at 1 0>]}}}
     Scale Curve=&FloatFromCurve(
      FloatCurve=DataInterface() {propertyValues {Curve {keys [
          <at 0 0.3>
          {arriveTangent=1.7182, interpMode=RCIM_Cubic, leaveTangent=1.7182, tangentMode=RCTM_User, time=0.1, value=0.3}
          {arriveTangent=0.5923, interpMode=RCIM_Cubic, leaveTangent=-0.0047, tangentMode=RCTM_Break, time=0.5, value=1}
          <at 1 1>
         ]}}}
      CurveIndex=$Emitter.Scalability.DistanceFraction
     )
    )
    Scale RGB=&Multiply_VectorByFloat(
     Vector=<1 1 1>
     Float=&FloatFromCurve(
      FloatCurve=DataInterface() {propertyValues {Curve {keys [
          {arriveTangent=-8.1251, interpMode=RCIM_Cubic, leaveTangent=-8.1251, tangentMode=RCTM_User, time=0, value=1}
          {arriveTangent=-0.0389, interpMode=RCIM_Cubic, leaveTangent=-0.0388, tangentMode=RCTM_User, time=0.2, value=0.1}
          <at 1 0.1>
         ]}}}
      Scale Curve=10
     )
    )
    ScaleA=true
    ScaleRGB=true
   )
   &ScaleSpriteSize(Scale Factor=&Vector2DFromFloat(Value=&FloatFromCurve(
      FloatCurve=DataInterface() {propertyValues {Curve {keys [<at 0 0.7 RCIM_Cubic>, {arriveTangent=1.4874, interpMode=RCIM_Cubic, leaveTangent=1.4874, tangentMode=RCTM_User, time=1, value=1.5}]}}}
      Scale Curve=&Lerp_Float(
       A=1
       B=2.5
       Alpha=&NormalizeFloat(Value=$Particles.VelFromPointRand, Min=0, Max=$Particles.ParticleVelocity)
      )
     )))
   &SolveForcesAndVelocity(Clamp Velocity=false, Limit Acceleration=false)
  ]
  rendererClasses [&NiagaraSpriteRendererProperties]
  renderers [{rendererClass=&NiagaraSpriteRendererProperties}]
 }
 Sparkles {
  emitterUpdateScript [
   &EmitterState(Enable Distance Culling=true, MaxDistance=7000, Module.bUseMaxDistance=true, Module.bUseMinDistance=false, Scalability Mode=::NewEnumerator1)
   &SpawnRate(SpawnRate=50, Use Spawn Probability=false)
  ]
  particleSpawnScript [
   &InitializeParticle(
    Color Maximum=<rgba 0.1262 1.7405 2 1>
    Color Minimum=<rgba 1.3639 0.2086 2 1>
    Color Mode=::NewEnumerator2
    Lifetime Max=4
    Lifetime Min=2
    Lifetime Mode=::NewEnumerator1
    Mesh Scale Mode=::NewEnumerator4
    Position Mode=::NewEnumerator1
    Sprite Rotation Mode=::NewEnumerator2
    Sprite Size Mode=::NewEnumerator1
    Uniform Sprite Size Max=90
    Uniform Sprite Size Min=30
    UsePositionOffset=false
    Write Material Random=true
   )
   &AddVelocityFromPoint(
    Module.Use Falloff Distance=false
    Origin Offset=<0 -1 0>
    Velocity Strength=&FloatFromCurve(
     FloatCurve=DataInterface() {propertyValues {Curve {keys [
         {interpMode=RCIM_Cubic, tangentMode=RCTM_User, time=0, value=0.1}
         {arriveTangent=2.7242, interpMode=RCIM_Cubic, leaveTangent=2.7242, tangentMode=RCTM_User, time=1, value=1}
        ]}}}
     CurveIndex=&RandomRangeFloat(Minimum=0.5, Maximum=1)
     Scale Curve=200
    )
   )
   &ShapeLocation(MaskSpawnGroup=false, Override Seed=false, Sphere Radius=35, UseOwnerScale=true)
  ]
  particleUpdateScript [
   &ParticleState()
   &Drag(Drag=0.5, Use Linear Drag=true, Use Rotational Drag=false)
   &CurlNoiseForce(
    Noise Strength=&FloatFromCurve(
     FloatCurve=DataInterface() {propertyValues {Curve {keys [
         {interpMode=RCIM_Cubic, tangentMode=RCTM_User, time=0, value=0}
         {arriveTangent=0.8922, interpMode=RCIM_Cubic, leaveTangent=0.8922, tangentMode=RCTM_User, time=1, value=1}
        ]}}}
     Scale Curve=500
    )
    Pan Noise=false
   )
   &DynamicMaterialParameters(
    Index 0 Param 1=1
    Index 0 Param 2=1
    Index 0 Param 3=0.5
    Module.Param0WriteEnabled=true
    Module.Param1WriteEnabled=true
    Module.Param2WriteEnabled=true
    Module.Param3WriteEnabled=true
   )
   &ScaleColor(
    Scale Alpha=&FloatFromCurve(
     FloatCurve=DataInterface() {propertyValues {Curve {keys [
         {arriveTangent=27.7425, interpMode=RCIM_Cubic, leaveTangent=27.7426, tangentMode=RCTM_User, time=0, value=0}
         {interpMode=RCIM_Cubic, tangentMode=RCTM_User, time=0.1, value=1}
         {arriveTangent=-2.0835, interpMode=RCIM_Cubic, leaveTangent=-2.0835, tangentMode=RCTM_User, time=0.8081, value=0.5934}
         <at 1 0>
        ]}}}
     Scale Curve=&FloatFromCurve(
      FloatCurve=DataInterface() {propertyValues {Curve {keys [
          <at 0 0.1>
          {arriveTangent=1.7182, interpMode=RCIM_Cubic, leaveTangent=1.7182, tangentMode=RCTM_User, time=0.1, value=0.1}
          {arriveTangent=0.5923, interpMode=RCIM_Cubic, leaveTangent=-0.0047, tangentMode=RCTM_Break, time=0.5, value=1}
          <at 1 1>
         ]}}}
      CurveIndex=$Emitter.Scalability.DistanceFraction
     )
    )
    Scale RGB=&VectorFromFloat(Value=&FloatFromCurve(
      FloatCurve=DataInterface() {propertyValues {Curve {keys [
          {arriveTangent=0.0361, interpMode=RCIM_Cubic, leaveTangent=0.0361, tangentMode=RCTM_User, time=0, value=1}
          {arriveTangent=-0.2536, interpMode=RCIM_Cubic, leaveTangent=-0.2536, tangentMode=RCTM_User, time=0.2, value=0.1}
          <at 1 0.05>
         ]}}}
      Scale Curve=25
     ))
    ScaleA=true
    ScaleRGB=true
   )
   &ScaleSpriteSize(Scale Factor=&Vector2DFromFloat(Value=&FloatFromCurve(
      FloatCurve=DataInterface() {propertyValues {Curve {keys [
          <at 0 0 RCIM_Cubic>
          {arriveTangent=0.836, interpMode=RCIM_Cubic, leaveTangent=0.836, time=0.2, value=1}
          {arriveTangent=-1.8455, interpMode=RCIM_Cubic, leaveTangent=-1.8455, tangentMode=RCTM_User, time=0.6507, value=0.544}
          <at 1 0>
         ]}}}
      Scale Curve=&FloatFromCurve(
       FloatCurve=DataInterface() {propertyValues {Curve {keys [
           <at 0 0.3>
           {arriveTangent=0.0785, interpMode=RCIM_Cubic, leaveTangent=0.0785, tangentMode=RCTM_User, time=0.2, value=0.3}
           {arriveTangent=0.5923, interpMode=RCIM_Cubic, leaveTangent=-0.0047, tangentMode=RCTM_Break, time=0.5, value=1}
           <at 1 1>
          ]}}}
       CurveIndex=$Emitter.Scalability.DistanceFraction
      )
     )))
   &SolveForcesAndVelocity(Clamp Velocity=false, Limit Acceleration=false)
  ]
  rendererClasses [&NiagaraSpriteRendererProperties]
  renderers [{rendererClass=&NiagaraSpriteRendererProperties}]
 }
 SparklesEnd {
  bEnabled=false
  emitterUpdateScript [
   &EmitterState(Enable Distance Culling=true, MaxDistance=7000, Module.bUseMaxDistance=true, Module.bUseMinDistance=false, Scalability Mode=::NewEnumerator1)
   &SpawnRate(SpawnRate=50, Use Spawn Probability=false)
  ]
  particleSpawnScript [
   &InitializeParticle(
    Color Maximum=<rgba 0.1262 1.7405 2 1>
    Color Minimum=<rgba 1.3639 0.2086 2 1>
    Color Mode=::NewEnumerator2
    Lifetime Mode=::NewEnumerator1
    Mesh Scale Mode=::NewEnumerator4
    Position=<0 500 0>
    Position Mode=::NewEnumerator1
    Sprite Rotation Mode=::NewEnumerator2
    Sprite Size Mode=::NewEnumerator1
    Uniform Sprite Size Max=90
    Uniform Sprite Size Min=30
    UsePositionOffset=false
    Write Material Random=true
   )
   &AddVelocityFromPoint(
    Module.Use Falloff Distance=false
    Origin Offset=<0 -10 0>
    Velocity Strength=&FloatFromCurve(
     FloatCurve=DataInterface() {propertyValues {Curve {keys [
         {interpMode=RCIM_Cubic, tangentMode=RCTM_User, time=0, value=0.1}
         {arriveTangent=2.7242, interpMode=RCIM_Cubic, leaveTangent=2.7242, tangentMode=RCTM_User, time=1, value=1}
        ]}}}
     CurveIndex=&RandomRangeFloat(Minimum=0.5, Maximum=1)
     Scale Curve=400
    )
   )
   &ShapeLocation(MaskSpawnGroup=false, Override Seed=false, Sphere Radius=35, UseOwnerScale=true)
  ]
  particleUpdateScript [
   &ParticleState()
   &Drag(Drag=0.5, Use Linear Drag=true, Use Rotational Drag=false)
   &DynamicMaterialParameters(
    Index 0 Param 1=1
    Index 0 Param 2=1
    Index 0 Param 3=0.5
    Module.Param0WriteEnabled=true
    Module.Param1WriteEnabled=true
    Module.Param2WriteEnabled=true
    Module.Param3WriteEnabled=true
   )
   &ScaleColor(
    Scale Alpha=&FloatFromCurve(
     FloatCurve=DataInterface() {propertyValues {Curve {keys [
         {arriveTangent=5.8739, interpMode=RCIM_Cubic, leaveTangent=5.8739, tangentMode=RCTM_User, time=0, value=0}
         {interpMode=RCIM_Cubic, tangentMode=RCTM_User, time=0.5, value=1}
         {arriveTangent=-2.0835, interpMode=RCIM_Cubic, leaveTangent=-2.0835, tangentMode=RCTM_User, time=0.8081, value=0.5934}
         <at 1 0>
        ]}}}
     Scale Curve=&FloatFromCurve(
      FloatCurve=DataInterface() {propertyValues {Curve {keys [
          <at 0 0.1>
          {arriveTangent=1.7182, interpMode=RCIM_Cubic, leaveTangent=1.7182, tangentMode=RCTM_User, time=0.1, value=0.1}
          {arriveTangent=0.5923, interpMode=RCIM_Cubic, leaveTangent=-0.0047, tangentMode=RCTM_Break, time=0.5, value=1}
          <at 1 1>
         ]}}}
      CurveIndex=$Emitter.Scalability.DistanceFraction
     )
    )
    Scale RGB=&VectorFromFloat(Value=&FloatFromCurve(
      FloatCurve=DataInterface() {propertyValues {Curve {keys [
          {arriveTangent=0.0361, interpMode=RCIM_Cubic, leaveTangent=0.0361, tangentMode=RCTM_User, time=0, value=1}
          {arriveTangent=-0.2536, interpMode=RCIM_Cubic, leaveTangent=-0.2536, tangentMode=RCTM_User, time=0.2, value=0.1}
          <at 1 0.05>
         ]}}}
      Scale Curve=25
     ))
    ScaleA=true
    ScaleRGB=true
   )
   &ScaleSpriteSize(Scale Factor=&Vector2DFromFloat(Value=&FloatFromCurve(
      FloatCurve=DataInterface() {propertyValues {Curve {keys [
          <at 0 0 RCIM_Cubic>
          {arriveTangent=0.836, interpMode=RCIM_Cubic, leaveTangent=0.836, time=0.2, value=1}
          {arriveTangent=-1.8455, interpMode=RCIM_Cubic, leaveTangent=-1.8455, tangentMode=RCTM_User, time=0.6507, value=0.544}
          <at 1 0>
         ]}}}
      Scale Curve=&FloatFromCurve(
       FloatCurve=DataInterface() {propertyValues {Curve {keys [
           <at 0 0.3>
           {arriveTangent=0.0785, interpMode=RCIM_Cubic, leaveTangent=0.0785, tangentMode=RCTM_User, time=0.2, value=0.3}
           {arriveTangent=0.5923, interpMode=RCIM_Cubic, leaveTangent=-0.0047, tangentMode=RCTM_Break, time=0.5, value=1}
           <at 1 1>
          ]}}}
       CurveIndex=$Emitter.Scalability.DistanceFraction
      )
     )))
   &SolveForcesAndVelocity(Clamp Velocity=false, Limit Acceleration=false)
  ]
  rendererClasses [&NiagaraSpriteRendererProperties]
  renderers [{rendererClass=&NiagaraSpriteRendererProperties}]
 }
}
summary {
 systemName=NS_FriedNexus_ParticleAccel_Beam
 emitters [
  {emitterName=Beam, rendererClasses [&NiagaraSpriteRendererProperties, &NiagaraRibbonRendererProperties]}
  {emitterName=Sparkles, rendererClasses [&NiagaraSpriteRendererProperties]}
  {bEnabled=false, emitterName=SparklesEnd, rendererClasses [&NiagaraSpriteRendererProperties]}
  {emitterName=InnerGlow, rendererClasses [&NiagaraSpriteRendererProperties]}
  {emitterName=OuterGlow, rendererClasses [&NiagaraSpriteRendererProperties, &NiagaraLightRendererProperties]}
  {emitterName=OuterGlow001, rendererClasses [&NiagaraSpriteRendererProperties, &NiagaraLightRendererProperties]}
  {emitterName=InnerGlow001, rendererClasses [&NiagaraSpriteRendererProperties]}
  {emitterName=Shimmers, rendererClasses [&NiagaraSpriteRendererProperties]}
 ]
 userVariables [User.InteriorColor {type=LinearColor, description="", defaultValue {struct=LinearColor, value=<rgba 1 1 1 1>}}]
}
