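# Niagara system for the holographic effect on the FriedNexus particle accelerator prop, in which only the RotatingRays emitter has content.
# RotatingRays bursts a small random count of scaled meshes, spread evenly around one rotation axis with slight random tilt.
# The meshes spin at random rates scaled by User.RotationSpeedScalar, take alpha from User.AlphaScalar, and persist past their lifetime.
# Flare, GlowStreaks and DarkBackdrop show no renderers or modules, and the dynamic material parameter module is disabled.
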NiagaraSystem /Game/Effects/Fort_Effects/Effects/Environment/Hera/Props/FriedNexus_ParticleAccelerator_A/NS_FriedNexus_ParticleAccelerator_Holo
dependencies {usedDataInterfaces [], usedDynamicInputs [], usedModules [&SystemState]}
emitters {
 DarkBackdrop {rendererClasses [], renderers []}
 Flare {rendererClasses [], renderers []}
 GlowStreaks {rendererClasses [], renderers []}
 RotatingRays {
  emitterSpawnScript [/SetVariables(Emitter.BurstNum:NiagaraInt32=&RandomRangeInt(Minimum=3, Maximum=5)) {isSetParameters=true}]
  emitterUpdateScript [
   &EmitterState()
   &SpawnBurst_Instantaneous(Spawn Count=$Emitter.BurstNum, Use Spawn Probability=false, UseLoopCountLimit=true)
  ]
  particleSpawnScript [
   &InitializeParticle(
    Color Mode=::NewEnumerator0
    Lifetime=5
    Mesh Scale=&Multiply_VectorByFloat(Vector=&RandomRangeVector(Minimum=<1 1 0.75>, Maximum=<1 1 1>), Float=7)
    Mesh Scale Mode=::NewEnumerator4
    UsePositionOffset=false
    Write Material Random=true
   )
   &InitialMeshOrientation(
    Mesh Orientation Mode=::NewEnumerator7
    Rotation=&MakeVector(
     X=&RandomRangeFloat(Minimum=-0.15, Maximum=0.15)
     Y=&ReturnNormalizedExecIndex(Normalized Index Includes One=false)
     Z=0
    )
    Use Rotation Vector=true
   )
   &MeshRotationForce(Rotation=&Multiply_VectorByFloat(Vector=&RandomRangeVector(Minimum=<0 -5 0>, Maximum=<0 -15 0>), Float=$User.RotationSpeedScalar))
   &ApplyInitialForces()
  ]
  particleUpdateScript [
   &ParticleState(Kill Particles When Lifetime Has Elapsed=false, Let Infinitely Lived Particles Die When Emitter Deactivates=true)
   &DynamicMaterialParameters(
    Index 0 Param 1=&RandomRangeFloat(Minimum=-2, Maximum=2)
    Index 0 Param 2=1
    Module.Index0Float0Write=true
    Module.Index0Float1Write=true
    Module.Index0Float2Write=true
    Module.Index0Float3Write=true
   ) {enabled=false}
   &SolveForcesAndVelocity(Clamp Velocity=false, Limit Acceleration=false)
   &ScaleColor(Scale Alpha=$User.AlphaScalar, ScaleA=true, ScaleRGB=false)
  ]
  rendererClasses [&NiagaraMeshRendererProperties]
  renderers [{rendererClass=&NiagaraMeshRendererProperties, rendererIndex=0}]
 }
}
summary {
 systemName=NS_FriedNexus_ParticleAccelerator_Holo
 emitters [
  {emitterName=Flare, rendererClasses []}
  {emitterName=RotatingRays, rendererClasses [&NiagaraMeshRendererProperties]}
  {emitterName=GlowStreaks, rendererClasses []}
  {emitterName=DarkBackdrop, rendererClasses []}
 ]
 userVariables [
  User.AlphaScalar {type=NiagaraFloat, defaultValue=2, description=""}
  User.RotationSpeedScalar {type=NiagaraFloat, defaultValue=0.75, description=""}
 ]
}
